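# Physics asset for the RoseForm backpack cosmetic, defining capsule bodies and joints for its dangling bullets, paper, pliers, screwdriver, strap, tag and wrench.
# Each item has a small base body with rotation locked to the three-capsule pack body.
# Dynamic bodies hang from those bases with narrow per-joint swing and twist limits.
# Bullets form a four-link chain with wider single-axis swing, while paper and tag use two-link chains.
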PhysicsAsset /BRCosmetics/Accessories/FORT_Backpacks/Backpack_F_MED_RoseForm/Meshes/F_MED_RoseForm_Pack_Physics
bodies {
 base_bullet [{shapeType=Capsule, length=1, radius=1, shapeName=base_bullet_capsule, center=<0 0 0.5715>, rotation=<rot -90 0 0>}]
 base_paper [{shapeType=Capsule, length=0.5, radius=0.5, shapeName=base_paper_capsule, rotation=<rot 90 0 0>}]
 base_pliers [{shapeType=Capsule, length=0, radius=0.5, shapeName=base_pliers_capsule, center=<-0.4346 0 0>, rotation=<rot 90 0 0>}]
 base_screwdriver [{shapeType=Capsule, length=1, radius=1, shapeName=base_screwdriver_capsule, rotation=<rot -90 0 0>}]
 base_strap [{shapeType=Capsule, length=0.25, radius=1, shapeName=base_strap_capsule, center=<0 0 0.8048>, rotation=<rot 90 0 0>}]
 base_tag [{shapeType=Capsule, length=0.5, radius=0.5, shapeName=base_tag_capsule, rotation=<rot 90 0 0>}]
 base_wrench [{shapeType=Capsule, length=0, radius=0.75, shapeName=base_wrench_capsule, rotation=<rot 90 0 0>}]
 dyn_bullet_1 [{shapeType=Capsule, length=4, radius=2.25, shapeName=dyn_bullet_1_capsule, center=<5.2014 0 0>, rotation=<rot -90 180 90.0167>}]
 dyn_bullet_2 [{shapeType=Capsule, length=3.5, radius=2.25, shapeName=dyn_bullet_2_capsule, center=<2.53 0 1.192>, rotation=<rot -90 -165.9638 75.9804>}]
 dyn_bullet_3 [{
   shapeType=Capsule
   length=3.8507
   radius=2.1493
   shapeName=dyn_bullet_3_capsule
   center=<3.1574 -0.6451 0.7779>
   rotation=<rot -90 -165.9638 75.9804>
  }]
 dyn_bullet_4 [{
   shapeType=Capsule
   length=3.5
   radius=2
   shapeName=dyn_bullet_4_capsule
   center=<2.8066 -0.9664 0.7161>
   rotation=<rot -90 -165.9638 75.9804>
  }]
 dyn_paper_1 [{shapeType=Capsule, length=2, radius=2, shapeName=dyn_paper_1_capsule, rotation=<rot -90 0 0>}]
 dyn_paper_2 [{shapeType=Capsule, length=2, radius=2.25, shapeName=dyn_paper_2_capsule, center=<4.267 0 -0.1136>, rotation=<rot -90 0 0>}]
 dyn_pliers_1 [{
   shapeType=Capsule
   length=7.5
   radius=2
   shapeName=dyn_pliers_1_capsule
   center=<0.1657 -0.94 -0.4604>
   rotation=<rot -69.9998 100.0038 -0.004>
  }]
 dyn_screwdriver1 [{shapeType=Capsule, length=2.5, radius=1.75, shapeName=dyn_screwdriver1_capsule, center=<1.5238 0 3.3023>}]
 dyn_strap_1 [{
   shapeType=Capsule
   length=1.5
   radius=1.5
   shapeName=dyn_strap_1_capsule
   center=<2.6535 0 -0.4212>
   rotation=<rot -79.3029 -179.9968 179.9964>
  }]
 dyn_tag_1 [{shapeType=Capsule, length=0.25, radius=0.5, shapeName=dyn_tag_1_capsule, center=<0.2605 0 0>, rotation=<rot -90 0 0>}]
 dyn_tag_2 [{shapeType=Capsule, length=2, radius=0.75, shapeName=dyn_tag_2_capsule, center=<1.6866 0 -0.9367>, rotation=<rot -49.9994 -180 179.9999>}]
 dyn_wrench_1 [{shapeType=Capsule, length=11.25, radius=1.5, shapeName=dyn_wrench_1_capsule, center=<5.128 0 -0.3973>, rotation=<rot -90 180 180>}]
 pack [
  {shapeType=Capsule, length=18.6538, radius=6.2552, shapeName=pack_capsule, center=<-1.2028 -6.7177 0.3638>, rotation=<rot 90 0 0>}
  {shapeType=Capsule, length=21.25, radius=5.25, shapeName=pack_capsule, center=<-0.14 -7.0491 -8.1424>, rotation=<rot -90 179.9208 -179.9211>}
  {
   shapeType=Capsule
   length=5.75
   radius=1.25
   shapeName=pack_capsule
   center=<-11.0226 -12.3872 -1.9131>
   rotation=<rot 5.2483 -0.719 -7.8122>
  }
 ]
}
constraints [
 {bone1Name=dyn_bullet_1, bone2Name=base_bullet, swing1LimitDegrees=2, swing2LimitDegrees=45, twistLimitDegrees=1}
 {bone1Name=dyn_bullet_2, bone2Name=dyn_bullet_1, swing1LimitDegrees=0, swing2LimitDegrees=30, twistLimitDegrees=0}
 {bone1Name=dyn_bullet_3, bone2Name=dyn_bullet_2, swing1LimitDegrees=0, swing2LimitDegrees=45, twistLimitDegrees=20}
 {bone1Name=dyn_bullet_4, bone2Name=dyn_bullet_3, swing1LimitDegrees=0, swing2LimitDegrees=50, twistLimitDegrees=0}
 {bone1Name=dyn_paper_1, bone2Name=base_paper, swing1LimitDegrees=2, swing2LimitDegrees=2, twistLimitDegrees=1}
 {bone1Name=dyn_pliers_1, bone2Name=base_pliers, swing1LimitDegrees=5, swing2LimitDegrees=5, twistLimitDegrees=1}
 {bone1Name=dyn_screwdriver1, bone2Name=base_screwdriver, swing1LimitDegrees=7, swing2LimitDegrees=10, twistLimitDegrees=1}
 {bone1Name=dyn_strap_1, bone2Name=base_strap, swing1LimitDegrees=20, swing2LimitDegrees=33, twistLimitDegrees=3}
 {bone1Name=dyn_tag_1, bone2Name=base_tag, swing1LimitDegrees=1, swing2LimitDegrees=10, twistLimitDegrees=1}
 {bone1Name=dyn_tag_2, bone2Name=dyn_tag_1, swing1LimitDegrees=10, swing2LimitDegrees=10, twistLimitDegrees=1}
 {bone1Name=dyn_wrench_1, bone2Name=base_wrench, swing1LimitDegrees=1, swing2LimitDegrees=5, twistLimitDegrees=1}
 {bone1Name=dyn_paper_2, bone2Name=dyn_paper_1, swing1LimitDegrees=3, swing2LimitDegrees=3, twistLimitDegrees=1}
 {
  bone1Name=base_bullet
  bone2Name=pack
  swing1LimitDegrees=45
  swing1Motion=Locked
  swing2LimitDegrees=45
  swing2Motion=Locked
  twistLimitDegrees=45
  twistMotion=Locked
 }
 {
  bone1Name=base_paper
  bone2Name=pack
  swing1LimitDegrees=45
  swing1Motion=Locked
  swing2LimitDegrees=45
  swing2Motion=Locked
  twistLimitDegrees=45
  twistMotion=Locked
 }
 {
  bone1Name=base_pliers
  bone2Name=pack
  swing1LimitDegrees=45
  swing1Motion=Locked
  swing2LimitDegrees=45
  swing2Motion=Locked
  twistLimitDegrees=45
  twistMotion=Locked
 }
 {
  bone1Name=base_strap
  bone2Name=pack
  swing1LimitDegrees=45
  swing1Motion=Locked
  swing2LimitDegrees=45
  swing2Motion=Locked
  twistLimitDegrees=45
  twistMotion=Locked
 }
 {
  bone1Name=base_tag
  bone2Name=pack
  swing1LimitDegrees=45
  swing1Motion=Locked
  swing2LimitDegrees=45
  swing2Motion=Locked
  twistLimitDegrees=45
  twistMotion=Locked
 }
 {
  bone1Name=base_wrench
  bone2Name=pack
  swing1LimitDegrees=45
  swing1Motion=Locked
  swing2LimitDegrees=45
  swing2Motion=Locked
  twistLimitDegrees=45
  twistMotion=Locked
 }
 {
  bone1Name=base_screwdriver
  bone2Name=pack
  swing1LimitDegrees=45
  swing1Motion=Locked
  swing2LimitDegrees=45
  swing2Motion=Locked
  twistLimitDegrees=45
  twistMotion=Locked
 }
]
